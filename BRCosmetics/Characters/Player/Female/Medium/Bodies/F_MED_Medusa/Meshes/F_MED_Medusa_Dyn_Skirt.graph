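# Physics asset for the dynamic skirt of the female medium Medusa player body, simulating skirt panels against the pelvis and thighs.
# Box bodies form short panel chains: back panels hang from the pelvis, the center via a master bone, and side panels from thigh twist bones.
# Chain joints use narrow twist limits, while angularly free constraints cross-link the lower back and right-side segments.
# The left side panel is one longer box without a cross-link.
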PhysicsAsset /BRCosmetics/Characters/Player/Female/Medium/Bodies/F_MED_Medusa/Meshes/F_MED_Medusa_Dyn_Skirt
bodies {
 dyn_skirt_bk_01 [{shapeType=Box, extentX=13.375, extentY=15.3125, extentZ=3.625, center=<6.708 0 0>}]
 dyn_skirt_bk_02 [{shapeType=Box, extentX=16, extentY=15.3125, extentZ=3.375, center=<7.0912 0 0>}]
 dyn_skirt_bk_MSTR [{shapeType=Capsule, length=2, radius=2, rotation=<rot 90 0 0>}]
 dyn_skirt_bk_l_01 [{shapeType=Box, extentX=13.375, extentY=15.3125, extentZ=4.2882, center=<6.708 0 0>}]
 dyn_skirt_bk_l_02 [{shapeType=Box, extentX=13.375, extentY=15.3125, extentZ=4.2882, center=<6.708 0 0>}]
 dyn_skirt_bk_r_01 [{shapeType=Box, extentX=13.375, extentY=15.3125, extentZ=1.875, center=<6.708 0 0>}]
 dyn_skirt_bk_r_02 [{shapeType=Box, extentX=13.375, extentY=15.3125, extentZ=4.0009, center=<6.708 0 0>}]
 dyn_skirt_l_01 [{shapeType=Box, extentX=24.3473, extentY=15.3125, extentZ=1.875, center=<10.9754 0 -0.25>, rotation=<rot -2 1.2975 -15.0591>}]
 dyn_skirt_r_01 [{shapeType=Box, extentX=13.375, extentY=15.3125, extentZ=1.875, center=<6.708 0 0>}]
 dyn_skirt_r_02 [{shapeType=Box, extentX=13.375, extentY=15.3125, extentZ=1.875, center=<6.708 0 0>}]
 pelvis [{shapeType=Capsule, length=2, radius=2, rotation=<rot 90 0 0>}]
 thigh_l [{shapeType=Capsule, length=24.9348, radius=8.9688, center=<-22.4376 -0.7113 0.9926>, rotation=<rot 82.4329 177.9388 -176.7435>}]
 thigh_r [{shapeType=Capsule, length=24.9348, radius=8.9688, center=<22.4376 0.7113 -0.9926>, rotation=<rot 82.4329 177.9388 -176.7435>}]
 thigh_twist_01_l [{shapeType=Capsule, length=2.3438, radius=1.6562, rotation=<rot 90 0 0>}]
 thigh_twist_01_r [{shapeType=Capsule, length=2.3438, radius=1.6562, rotation=<rot 90 0 0>}]
}
constraints [
 {bone1Name=dyn_skirt_bk_l_01, bone2Name=pelvis, swing1LimitDegrees=10, swing2LimitDegrees=45, twistLimitDegrees=5}
 {bone1Name=dyn_skirt_bk_l_02, bone2Name=dyn_skirt_bk_l_01, swing1LimitDegrees=45, swing2LimitDegrees=45, twistLimitDegrees=5}
 {
  bone1Name=thigh_twist_01_l
  bone2Name=thigh_l
  swing1LimitDegrees=45
  swing1Motion=Free
  swing2LimitDegrees=45
  swing2Motion=Free
  twistLimitDegrees=45
  twistMotion=Free
 }
 {bone1Name=dyn_skirt_l_01, bone2Name=thigh_twist_01_l, swing1LimitDegrees=45, swing2LimitDegrees=45, twistLimitDegrees=5}
 {
  bone1Name=thigh_twist_01_r
  bone2Name=thigh_r
  swing1LimitDegrees=45
  swing1Motion=Free
  swing2LimitDegrees=45
  swing2Motion=Free
  twistLimitDegrees=45
  twistMotion=Free
 }
 {bone1Name=dyn_skirt_r_01, bone2Name=thigh_twist_01_r, swing1LimitDegrees=45, swing2LimitDegrees=45, twistLimitDegrees=5}
 {bone1Name=dyn_skirt_r_02, bone2Name=dyn_skirt_r_01, swing1LimitDegrees=45, swing2LimitDegrees=45, twistLimitDegrees=5}
 {bone1Name=dyn_skirt_bk_r_01, bone2Name=pelvis, swing1LimitDegrees=10, swing2LimitDegrees=45, twistLimitDegrees=5}
 {bone1Name=dyn_skirt_bk_r_02, bone2Name=dyn_skirt_bk_r_01, swing1LimitDegrees=45, swing2LimitDegrees=45, twistLimitDegrees=5}
 {
  bone1Name=dyn_skirt_bk_MSTR
  bone2Name=pelvis
  swing1LimitDegrees=45
  swing1Motion=Free
  swing2LimitDegrees=45
  swing2Motion=Free
  twistLimitDegrees=45
  twistMotion=Free
 }
 {bone1Name=dyn_skirt_bk_01, bone2Name=dyn_skirt_bk_MSTR, swing1LimitDegrees=45, swing2LimitDegrees=45, twistLimitDegrees=5}
 {bone1Name=dyn_skirt_bk_02, bone2Name=dyn_skirt_bk_01, swing1LimitDegrees=45, swing2LimitDegrees=45, twistLimitDegrees=5}
 {
  bone1Name=thigh_l
  bone2Name=pelvis
  swing1LimitDegrees=45
  swing1Motion=Free
  swing2LimitDegrees=45
  swing2Motion=Free
  twistLimitDegrees=45
  twistMotion=Free
 }
 {
  bone1Name=thigh_r
  bone2Name=pelvis
  swing1LimitDegrees=45
  swing1Motion=Free
  swing2LimitDegrees=45
  swing2Motion=Free
  twistLimitDegrees=45
  twistMotion=Free
 }
 {
  bone1Name=dyn_skirt_bk_r_02
  bone2Name=dyn_skirt_bk_02
  swing1LimitDegrees=45
  swing1Motion=Free
  swing2LimitDegrees=45
  swing2Motion=Free
  twistLimitDegrees=45
  twistMotion=Free
 }
 {
  bone1Name=dyn_skirt_bk_l_02
  bone2Name=dyn_skirt_bk_02
  swing1LimitDegrees=45
  swing1Motion=Free
  swing2LimitDegrees=45
  swing2Motion=Free
  twistLimitDegrees=45
  twistMotion=Free
 }
 {
  bone1Name=dyn_skirt_r_02
  bone2Name=dyn_skirt_bk_r_02
  swing1LimitDegrees=45
  swing1Motion=Free
  swing2LimitDegrees=45
  swing2Motion=Free
  twistLimitDegrees=45
  twistMotion=Free
 }
]
