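# Physics asset for the left-leg dynamic accessories of the Medusa medium female player body, covering calf strings, canteen, coiled whip and skirt.
# Capsule bodies sit on pelvis, thigh, calf twist and the dyn bones, with pelvis-to-thigh, calf, skirt and whip-canteen master joints left angularly free.
# Calf strings hang off the calf twist, while canteen, ropes and two-link whip handle and tip chains branch from the master with limited swing and tight twist.
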PhysicsAsset /BRCosmetics/Characters/Player/Female/Medium/Bodies/F_MED_Medusa/Meshes/F_MED_Medusa_Dyn_leg_l
bodies {
 calf_twist_02_l [{shapeType=Capsule, length=1, radius=1, rotation=<rot 90 0 0>}]
 dyn_calf_string_in_l [{shapeType=Capsule, length=3.4062, radius=1, center=<2.1755 0 0.0003>, rotation=<rot 90 0 0>}]
 dyn_calf_string_out_l [{shapeType=Capsule, length=3.4062, radius=1, center=<0.5914 0 -0.0003>, rotation=<rot 90 0 0>}]
 dyn_canteen [{shapeType=Capsule, length=1.875, radius=1, center=<-0.2 0 0>, rotation=<rot 90 0 0>}]
 dyn_skirt_l_01 [{
   shapeType=Capsule
   length=30.6875
   radius=14.0312
   shapeName=dyn_skirt_l_01_capsule
   center=<2.3182 1.3539 -14.6223>
   rotation=<rot -59.1042 105.2341 -103.1533>
  }]
 dyn_whip_bottom_out_rope [{
   shapeType=Capsule
   length=4.5312
   radius=2.75
   shapeName=dyn_whip_bottom_under_rope_capsule
   center=<11.3265 -0.0002 -0.2306>
   rotation=<rot 80 0 -0.0001>
  }]
 dyn_whip_bottom_under_rope [{
   shapeType=Capsule
   length=4.0938
   radius=2.75
   shapeName=dyn_whip_bottom_under_rope_capsule
   center=<11.3265 0 0.7718>
   rotation=<rot 80 0 -0.0001>
  }]
 dyn_whip_canteen_MSTR [{shapeType=Capsule, length=1, radius=1, rotation=<rot 90 0 0>}]
 dyn_whip_handle_fr_01 [{shapeType=Capsule, length=4.9211, radius=1.75, center=<8.6645 0.0001 0>, rotation=<rot 90 0 0>}]
 dyn_whip_handle_fr_02 [
  {shapeType=Capsule, length=3.125, radius=1.4688, center=<2.0653 0 0>, rotation=<rot 90 0 0>}
  {shapeType=Capsule, length=5.5312, radius=0.9688, center=<2.5191 0.4967 -0.6518>, rotation=<rot 71.1792 -22.6927 -50.8083>}
 ]
 dyn_whip_tip_bk_01 [{shapeType=Capsule, length=1, radius=3, center=<4.5 1.5 0>}]
 dyn_whip_tip_bk_02 [{shapeType=Capsule, length=6, radius=2, center=<5 1 -2>, rotation=<rot 59.9998 -0.0002 -0.0001>}]
 pelvis [{shapeType=Capsule, length=1, radius=1, rotation=<rot 90 0 0>}]
 thigh_l [
  {shapeType=Capsule, length=13.9375, radius=8.5625, center=<-11.3485 0 0>, rotation=<rot 90 0 0>}
  {shapeType=Capsule, length=6.75, radius=6.2188, center=<-21.0092 3.1957 -1.5958>, rotation=<rot 90 -12.3391 -16.2652>}
 ]
}
constraints [
 {bone1Name=dyn_calf_string_out_l, bone2Name=calf_twist_02_l, swing1LimitDegrees=45, swing2LimitDegrees=30, twistLimitDegrees=2}
 {bone1Name=dyn_calf_string_in_l, bone2Name=calf_twist_02_l, swing1LimitDegrees=45, swing2LimitDegrees=30, twistLimitDegrees=2}
 {bone1Name=dyn_canteen, bone2Name=dyn_whip_canteen_MSTR, swing1LimitDegrees=45, swing2LimitDegrees=45, twistLimitDegrees=10}
 {bone1Name=dyn_whip_bottom_out_rope, bone2Name=dyn_whip_canteen_MSTR, swing1LimitDegrees=10, swing2LimitDegrees=45, twistLimitDegrees=2}
 {bone1Name=dyn_whip_tip_bk_01, bone2Name=dyn_whip_canteen_MSTR, swing1LimitDegrees=30, swing2LimitDegrees=30, twistLimitDegrees=2}
 {bone1Name=dyn_whip_handle_fr_01, bone2Name=dyn_whip_canteen_MSTR, swing1LimitDegrees=45, swing2LimitDegrees=45, twistLimitDegrees=5}
 {bone1Name=dyn_whip_bottom_under_rope, bone2Name=dyn_whip_canteen_MSTR, swing1LimitDegrees=10, swing2LimitDegrees=45, twistLimitDegrees=2}
 {bone1Name=dyn_whip_tip_bk_02, bone2Name=dyn_whip_tip_bk_01, swing1LimitDegrees=30, swing2LimitDegrees=30, twistLimitDegrees=2}
 {bone1Name=dyn_whip_handle_fr_02, bone2Name=dyn_whip_handle_fr_01, swing1LimitDegrees=30, swing2LimitDegrees=30, twistLimitDegrees=2}
 {
  bone1Name=calf_twist_02_l
  bone2Name=thigh_l
  swing1LimitDegrees=45
  swing1Motion=Free
  swing2LimitDegrees=45
  swing2Motion=Free
  twistLimitDegrees=45
  twistMotion=Free
 }
 {
  bone1Name=dyn_whip_canteen_MSTR
  bone2Name=thigh_l
  swing1LimitDegrees=45
  swing1Motion=Free
  swing2LimitDegrees=45
  swing2Motion=Free
  twistLimitDegrees=45
  twistMotion=Free
 }
 {
  bone1Name=thigh_l
  bone2Name=pelvis
  swing1LimitDegrees=45
  swing1Motion=Free
  swing2LimitDegrees=45
  swing2Motion=Free
  twistLimitDegrees=45
  twistMotion=Free
 }
 {
  bone1Name=dyn_skirt_l_01
  bone2Name=thigh_l
  swing1LimitDegrees=45
  swing1Motion=Free
  swing2LimitDegrees=45
  swing2Motion=Free
  twistLimitDegrees=45
  twistMotion=Free
 }
]
