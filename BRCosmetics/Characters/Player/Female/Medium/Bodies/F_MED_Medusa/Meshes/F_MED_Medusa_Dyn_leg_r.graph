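# Physics asset for the F_MED_Medusa character's right leg, covering two dynamic calf string bones for secondary motion.
# Small capsule bodies sit on the pelvis and calf_twist_02_r, with longer capsules on the inner and outer string bones.
# Both strings constrain directly to the calf twist bone with limited swing and very tight twist, while the calf twist links to the pelvis with free swing and twist.
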PhysicsAsset /BRCosmetics/Characters/Player/Female/Medium/Bodies/F_MED_Medusa/Meshes/F_MED_Medusa_Dyn_leg_r
bodies {
 calf_twist_02_r [{shapeType=Capsule, length=1, radius=1, rotation=<rot 90 0 0>}]
 dyn_calf_string_in_r [{shapeType=Capsule, length=3.4062, radius=1, center=<-0.7843 0 0.0002>, rotation=<rot 90 0 0>}]
 dyn_calf_string_out_r [{shapeType=Capsule, length=3.4062, radius=1, center=<-1.5519 0 0.0002>, rotation=<rot 90 0 0>}]
 pelvis [{shapeType=Capsule, length=1, radius=1, rotation=<rot 90 0 0>}]
}
constraints [
 {bone1Name=dyn_calf_string_out_r, bone2Name=calf_twist_02_r, swing1LimitDegrees=45, swing2LimitDegrees=30, twistLimitDegrees=2}
 {bone1Name=dyn_calf_string_in_r, bone2Name=calf_twist_02_r, swing1LimitDegrees=45, swing2LimitDegrees=30, twistLimitDegrees=2}
 {
  bone1Name=calf_twist_02_r
  bone2Name=pelvis
  swing1LimitDegrees=45
  swing1Motion=Free
  swing2LimitDegrees=45
  swing2Motion=Free
  twistLimitDegrees=45
  twistMotion=Free
 }
]
