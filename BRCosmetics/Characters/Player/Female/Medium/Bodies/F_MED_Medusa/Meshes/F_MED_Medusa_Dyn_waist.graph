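# Physics asset for the waist dynamic bones of the F_MED_Medusa female character body, covering the waist shirt sleeves and a back belt strap.
# Each sleeve is a two-capsule chain hung from the pelvis with limited swing and tight twist, while the strap attaches directly to the pelvis with narrower limits.
# Large thigh capsules, joined to the pelvis with free swing and twist, provide leg volumes alongside the dangling pieces.
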PhysicsAsset /BRCosmetics/Characters/Player/Female/Medium/Bodies/F_MED_Medusa/Meshes/F_MED_Medusa_Dyn_waist
bodies {
 dyn_back_belt_strap [{shapeType=Capsule, length=4.5625, radius=1, center=<0.1 0 0>, rotation=<rot 90 0 0>}]
 dyn_waist_shirt_sleeve_l_01 [{shapeType=Capsule, length=1, radius=1, center=<0.1 0 0>, rotation=<rot 90 0 0>}]
 dyn_waist_shirt_sleeve_l_02 [{shapeType=Capsule, length=4, radius=2.9062, center=<2.5611 0 0.0001>, rotation=<rot 90 0 0>}]
 dyn_waist_shirt_sleeve_r_01 [{shapeType=Capsule, length=1, radius=1, center=<0.1 0 0>, rotation=<rot 90 0 0>}]
 dyn_waist_shirt_sleeve_r_02 [{shapeType=Capsule, length=4.2812, radius=3.0625, center=<0.9898 0 0>, rotation=<rot 90 0 0>}]
 pelvis [{shapeType=Capsule, length=1, radius=1, rotation=<rot 90 0 0>}]
 thigh_l [{shapeType=Capsule, length=18.8125, radius=7.8438, center=<-11.3485 0 0>, rotation=<rot 90 0 0>}]
 thigh_r [{shapeType=Capsule, length=18.8125, radius=7.8438, center=<11.3485 0 0>, rotation=<rot 90 0 0>}]
}
constraints [
 {
  bone1Name=dyn_waist_shirt_sleeve_l_02
  bone2Name=dyn_waist_shirt_sleeve_l_01
  swing1LimitDegrees=45
  swing2LimitDegrees=45
  twistLimitDegrees=5
 }
 {
  bone1Name=dyn_waist_shirt_sleeve_r_02
  bone2Name=dyn_waist_shirt_sleeve_r_01
  swing1LimitDegrees=45
  swing2LimitDegrees=45
  twistLimitDegrees=5
 }
 {bone1Name=dyn_waist_shirt_sleeve_l_01, bone2Name=pelvis, swing1LimitDegrees=45, swing2LimitDegrees=45, twistLimitDegrees=5}
 {bone1Name=dyn_waist_shirt_sleeve_r_01, bone2Name=pelvis, swing1LimitDegrees=45, swing2LimitDegrees=45, twistLimitDegrees=5}
 {bone1Name=dyn_back_belt_strap, bone2Name=pelvis, swing1LimitDegrees=10, swing2LimitDegrees=45, twistLimitDegrees=2}
 {
  bone1Name=thigh_l
  bone2Name=pelvis
  swing1LimitDegrees=45
  swing1Motion=Free
  swing2LimitDegrees=45
  swing2Motion=Free
  twistLimitDegrees=45
  twistMotion=Free
 }
 {
  bone1Name=thigh_r
  bone2Name=pelvis
  swing1LimitDegrees=45
  swing1Motion=Free
  swing2LimitDegrees=45
  swing2Motion=Free
  twistLimitDegrees=45
  twistMotion=Free
 }
]
