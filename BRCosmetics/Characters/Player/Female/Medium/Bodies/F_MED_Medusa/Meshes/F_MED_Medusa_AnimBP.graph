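# Post-style Anim Blueprint for the F_MED_Medusa cosmetic body, adding secondary motion to its skirt, waist and leg accessories.
# It copies the attached parent's pose, then constrains half-thigh, canteen, skirt, sleeve and strap bones.
# Four chained RigidBody nodes simulate dangling parts, with skydiving force feeding the skirt.
# Additive bone offsets adjust the canteen when crouching or rigid bodies are disabled, and rear skirt bones by sprint level.
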AnimBlueprint /BRCosmetics/Characters/Player/Female/Medium/Bodies/F_MED_Medusa/Meshes/F_MED_Medusa_AnimBP
target_skeleton=&/F_MED_Medusa_Skeleton
graphs [
 AnimGraph {nodes [
   n1 = "Constraint - thigh_half_r (R)"(ComponentPose=Local To Component(LocalPose=Copy Pose From Mesh() {bUseAttachedParent=true}), Alpha=0.3, ConstraintWeights_0=1) {
    alphaBoolBlend {blendOption=Linear}
    boneToModify {boneName=thigh_half_r}
    constraintSetup [{transformType=Rotation, targetBone {boneName=thigh_r}}]
    constraintWeights [1]
   }
   n2 = "Constraint - thigh_half_l (R)"(ComponentPose=@n1, Alpha=0.3, ConstraintWeights_0=1) {
    alphaBoolBlend {blendOption=Linear}
    boneToModify {boneName=thigh_half_l}
    constraintSetup [{transformType=Rotation, targetBone {boneName=thigh_l}}]
    constraintWeights [1]
   }
   n3 = "Constraint - dyn_whip_canteen_MSTR (PPP)"!(ComponentPose=@n2, Alpha=1, ConstraintWeights_0=0.5182, ConstraintWeights_1:float=0.4391, ConstraintWeights_2:float=0.0428) -> (Pose) {
    alphaBoolBlend {blendOption=Linear}
    boneToModify {boneName=dyn_whip_canteen_MSTR}
    constraintSetup [{targetBone {boneName=pelvis}}, {targetBone {boneName=thigh_twist_01_l}}, {targetBone {boneName=thigh_l}}]
    constraintWeights [1,1,1]
   }
   n4 = "Constraint - dyn_skirt_bk_MSTR (PPP)"!(ComponentPose=@n3, Alpha=1, ConstraintWeights_0=0.9503, ConstraintWeights_1:float=0.0288, ConstraintWeights_2:float=0.0209) -> (Pose) {
    alphaBoolBlend {blendOption=Linear}
    boneToModify {boneName=dyn_skirt_bk_MSTR}
    constraintSetup [{targetBone {boneName=pelvis}}, {targetBone {boneName=thigh_twist_01_l}}, {targetBone {boneName=thigh_twist_01_r}}]
    constraintWeights [1,1,1]
   }
   n5 = "Constraint - dyn_waist_shirt_sleeve_l_01 (PPP)"!(ComponentPose=@n4, Alpha=1, ConstraintWeights_0=0.8429, ConstraintWeights_1:float=0.0859, ConstraintWeights_2:float=0.0712) -> (Pose) {
    alphaBoolBlend {blendOption=Linear}
    boneToModify {boneName=dyn_waist_shirt_sleeve_l_01}
    constraintSetup [{targetBone {boneName=pelvis}}, {targetBone {boneName=spine_02}}, {targetBone {boneName=spine_01}}]
    constraintWeights [1,1,1]
   }
   n6 = "Constraint - dyn_waist_shirt_sleeve_r_01 (PPP)"!(ComponentPose=@n5, Alpha=1, ConstraintWeights_0=0.8429, ConstraintWeights_1:float=0.0859, ConstraintWeights_2:float=0.0712) -> (Pose) {
    alphaBoolBlend {blendOption=Linear}
    boneToModify {boneName=dyn_waist_shirt_sleeve_r_01}
    constraintSetup [{targetBone {boneName=pelvis}}, {targetBone {boneName=spine_02}}, {targetBone {boneName=spine_01}}]
    constraintWeights [1,1,1]
   }
   n7 = "Constraint - dyn_back_belt_strap (PP)"!(ComponentPose=@n6, Alpha=1, ConstraintWeights_0=0.932, ConstraintWeights_1:float=0.068) -> (Pose) {
    alphaBoolBlend {blendOption=Linear}
    boneToModify {boneName=dyn_back_belt_strap}
    constraintSetup [{targetBone {boneName=pelvis}}, {targetBone {boneName=thigh_twist_01_r}}]
    constraintWeights [1,1]
   }
   n8 = "Transform (Modify) Bone - Bone: dyn_whip_canteen_MSTR"(
    ComponentPose=@n7
    Alpha=Get crouching alpha() -> (crouching alpha:double)
    Translation="-0.043,-0.118,1.435"
    Rotation="0,-25,-25"
    Scale="1,1,1"
   ) {
    rotationMode=BMM_Additive
    rotationSpace=BCS_BoneSpace
    translationMode=BMM_Additive
    translationSpace=BCS_BoneSpace
    alphaBoolBlend {blendOption=Linear}
    alphaScaleBiasClamp {bInterpResult=true}
    boneToModify {boneName=dyn_whip_canteen_MSTR}
    rotation {roll=-25, yaw=-25}
    translation=<-0.043 -0.118 1.435>
   }
   n9 = RigidBody(
    ComponentPose=@n8
    ExternalForce=Break Fort Anim Input Skydiving External Force(FortAnimInput_SkydivingExternalForce=Get SkydivingExternalForceInput()) -> (
     bUseSkydivingVectorForce=~
     SkydivingVectorForce=~
     HeadToPelvisDirection=~
     FloatingMultiplier=~
     FloatingAdditive=~
     DivingMultiplier=~
     DivingAdditive=~
     ParachutingMultiplier=~
     ParachutingAdditive=~
    ).SkydivingVectorForce
   ) {
    bClampLinearTranslationLimitToRefPose=true
    bOverrideWorldGravity=true
    overridePhysicsAsset=&/F_MED_Medusa_Dyn_Skirt
    simulationSpace=BaseBoneSpace
    alphaBoolBlend {blendOption=Linear}
    baseBoneRef {boneName=spine_01}
    componentAppliedLinearAccClamp=<2000 2000 2000>
    componentLinearAccScale=<0.4 0.4 0.4>
    componentLinearVelScale=<0.4 0.4 0.4>
    overrideWorldGravity {z=-980}
   }
   n10 = RigidBody(ComponentPose=@n9) {
    bClampLinearTranslationLimitToRefPose=true
    bOverrideWorldGravity=true
    overridePhysicsAsset=&/F_MED_Medusa_Dyn_waist
    simulationSpace=BaseBoneSpace
    alphaBoolBlend {blendOption=Linear}
    baseBoneRef {boneName=spine_01}
    componentAppliedLinearAccClamp=<2000 2000 2000>
    componentLinearAccScale=<0.3 0.1 0.1>
    componentLinearVelScale=<0.3 0.1 0.1>
    overrideWorldGravity {z=-980}
   }
   n11 = RigidBody(ComponentPose=@n10) {
    bClampLinearTranslationLimitToRefPose=true
    bOverrideWorldGravity=true
    overridePhysicsAsset=&/F_MED_Medusa_Dyn_leg_l
    simulationSpace=BaseBoneSpace
    alphaBoolBlend {blendOption=Linear}
    baseBoneRef {boneName=thigh_half_l}
    componentAppliedLinearAccClamp=<2000 2000 2000>
    componentLinearAccScale=<0.1 0.1 0.1>
    componentLinearVelScale=<0.1 0.1 0.1>
    overrideWorldGravity {z=-980}
   }
   n12 = RigidBody(ComponentPose=@n11) {
    bClampLinearTranslationLimitToRefPose=true
    bOverrideWorldGravity=true
    overridePhysicsAsset=&/F_MED_Medusa_Dyn_leg_r
    simulationSpace=BaseBoneSpace
    alphaBoolBlend {blendOption=Linear}
    baseBoneRef {boneName=thigh_half_r}
    componentAppliedLinearAccClamp=<2000 2000 2000>
    componentLinearAccScale=<0.1 0.1 0.1>
    componentLinearVelScale=<0.1 0.1 0.1>
    overrideWorldGravity {z=-980}
   }
   n13 = "Transform (Modify) Bone - Bone: dyn_whip_canteen_MSTR"(
    ComponentPose=@n12
    Alpha=Get rigidBody alpha() -> (rigidBody alpha:double)
    Translation="0,0,0"
    Rotation="15.543,0.000,0.000"
    Scale="1,1,1"
   ) {
    rotationMode=BMM_Additive
    rotationSpace=BCS_BoneSpace
    translationMode=BMM_Additive
    translationSpace=BCS_BoneSpace
    alphaBoolBlend {blendOption=Linear}
    alphaScaleBiasClamp {bInterpResult=true}
    boneToModify {boneName=dyn_whip_canteen_MSTR}
    rotation {pitch=15.543}
   }
   n14 = Get sprint() -> (sprint:double)
   n15 = "Transform (Modify) Bone - Bone: dyn_skirt_bk_r_01"(ComponentPose=@n13, Alpha=@n14, Translation="0,0,0", Rotation="-16.588,26.238,-18.511", Scale="1,1,1") {
    rotationMode=BMM_Additive
    rotationSpace=BCS_ParentBoneSpace
    translationMode=BMM_Additive
    translationSpace=BCS_ParentBoneSpace
    alphaBoolBlend {blendOption=Linear}
    boneToModify {boneName=dyn_skirt_bk_r_01}
    rotation=<rot -16.588 26.238 -18.511>
   }
   n16 = "Transform (Modify) Bone - Bone: dyn_skirt_bk_l_01"(ComponentPose=@n15, Alpha=@n14, Translation="0,0,0", Rotation="12.963,12.158,9.427", Scale="1,1,1") {
    rotationMode=BMM_Additive
    rotationSpace=BCS_ParentBoneSpace
    translationMode=BMM_Additive
    translationSpace=BCS_ParentBoneSpace
    alphaBoolBlend {blendOption=Linear}
    boneToModify {boneName=dyn_skirt_bk_l_01}
    rotation=<rot 12.963 12.158 9.427>
   }
   Output Pose(Result=Component To Local(ComponentPose=@n16)) {layerGroup=None}
  ]}
 EventGraph {nodes [
   Event BlueprintUpdateAnimation()
   > n1 = Branch(Condition=Get bRigidBodyEnabled())
   > n2 = Set rigidBody alpha(rigidBody alpha:double=0) -> (Output_Get:double)
   n3 = Branch(execute=[@n5.then, @n2.then], Condition=Get bIsSprinting())
   > n4 = Branch(Condition=Get bIsTacticalSprinting())
   > Set sprint(sprint:double=1) -> (Output_Get:double)
   Set sprint(execute=@n4.else, sprint=0.3)
   Set sprint(execute=@n3.else, sprint=0)
   n5 = Set rigidBody alpha(execute=@n1.else, rigidBody alpha=1)
   Event OnBeginCrouching()
   > Set crouching alpha(crouching alpha:double=1) -> (Output_Get:double)
   Event OnEndCrouching()
   > Set crouching alpha(crouching alpha=0)
  ]}
]
variables [
 crouching alpha {type=number}
 rigidBody alpha {type=number}
 sprint {type=number}
]
